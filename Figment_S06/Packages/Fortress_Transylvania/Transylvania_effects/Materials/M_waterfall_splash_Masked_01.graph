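# Masked, dithered particle-sprite material for waterfall splashes, built on a splash texture and its normal map.
# A slowly panning distortion map warps the UVs and joins the splash texture in a mask that dynamic parameters erode.
# Colour mixes a sun glint from global collection parameters, normal-distorted screen coordinates, optional particle colour and foam-like darkening masks.
# Mobile shading paths skip the distortion and depth fade, using plain particle colour and alpha.
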Material /Figment_S06/Packages/Fortress_Transylvania/Transylvania_effects/Materials/M_waterfall_splash_Masked_01
graph {
 properties {
  bTangentSpaceNormal=false
  bUsedWithParticleSprites=true
  blendMode=BLEND_Masked
  ditherOpacityMask=true
  pixelDepthOffsetMode=PDOM_Legacy
  translucencyLightingMode=TLM_VolumetricPerVertexNonDirectional
  translucencyPass=MTP_BeforeDOF
  translucentSelfShadowDensityScale=0
  translucentSelfShadowSecondDensityScale=0
  translucentShadowDensityScale=0
  usedShadingModels=Default Lit
 }
 nodes [
  n1 = TextureSample(UVs=Panner(Coordinate=TextureCoordinate()) {speedY=-0.03}) {texture=/Figment_S06/Packages/Fortress_Transylvania/Haunted_Forest/Swamps/Textures/T_Tilling_Distortion_Map_01}
  n2 = ComponentMask(None=@n1.G) {g=true, r=true}
  n3 = TextureCoordinate()
  n4 = ShadingPathSwitch(Default=Add(A=Multiply(A=@n2) {constB=0.1}, B=@n3), Mobile=@n3) {desc=Migrated from FeatureLevelSwitch}
  n5 = TextureSample(UVs=@n4) {desc=o, texture=/Game/Effects/Fort_Effects/Textures/Water/T_water_splash_04}
  n6 = Multiply(A=@n5.B, B=@n1.G)
  n7 = DynamicParameter() {defaultValue=<rgba 0 1 0 1>, paramNames [Param1,Param2,VTiling,UTiling]}
  n8 = LinearInterpolate(Alpha=@n7.Param2) {constA=5, constAlpha=0}
  n9 = Clamp(None=Multiply(A=Power(Base=@n6, Exp=@n8) {constExponent=50}, B=LinearInterpolate(Alpha=@n7.Param2) {constAlpha=0, constB=200}))
  n10 = ParticleColor()
  Multiply(A=ShadingPathSwitch(Default=Multiply(A=@n5.A, B=@n9), Mobile=Multiply(A=@n5.A, B=@n10.A)) {desc=Migrated from FeatureLevelSwitch})
  n11 = CollectionParameter() {
   collection=&FortniteMaterialParameters
   parameterName=SunAndMoonModelDirectionalVector
  }
  n12 = TextureSample(UVs=@n4) {samplerType=SAMPLERTYPE_Normal, texture=/Game/Effects/Fort_Effects/Textures/Water/T_water_splash_04_n}
  n13 = Transform(None=@n12.RGB)
  n14 = Clamp(None=DotProduct(A=ComponentMask(None=@n11) {b=true, g=true, r=true}, B=ReflectionVectorWS(CustomWorldNormal=@n13)))
  n15 = ComponentMask(None=CollectionParameter() {collection=&FortniteMaterialParameters, parameterName=SunLightColor}) {b=true, g=true, r=true}
  n16 = Multiply(A=Clamp(None=Power(Base=@n14) {constExponent=20}) {maxDefault=5}, B=Clamp(None=Power(Base=@n15) {constExponent=0.1}))
  n17 = Constant() {r=5}
  ShadingPathSwitch(
   Default=MaterialFunctionCall(DiffuseColor=Add(A=@n16, B=@n17)) {materialFunction=&ConvertFromDiffSpec}.BaseColor
   Mobile=Constant()
  ) {desc=Migrated from FeatureLevelSwitch}
  n18 = ComponentMask(None=@n12.RGB) {g=true, r=true}
  n19 = ViewSize()
  n20 = SceneTexelSize()
  n21 = ConstantBiasScale(None=Divide(A=ScreenPosition().ViewportUV, B=Multiply(A=@n19, B=@n20))) {bias=-0.5, scale=2}
  n22 = OneMinus(None=Multiply(A=@n21, B=@n21))
  n23 = MaterialFunctionCall(Float2=Clamp(None=Multiply(A=@n22) {constB=0.125})) {materialFunction=&BreakOutFloat2Components}
  n24 = Add(
   A=ComponentMask(None=ScreenPosition().ViewportUV) {g=true, r=true}
   B=Multiply(A=Multiply(A=@n18) {constB=0.2}, B=Multiply(A=@n23.R, B=@n23.R))
  )
  n25 = Add(A=@n16, B=AppendVector(A=Clamp(None=@n24), B=ScalarParameter() {defaultValue=1}))
  n26 = Add(A=@n25, B=ParticleColor().RGB)
  n27 = Clamp(None=OneMinus(None=@n5.B))
  n28 = OneMinus(None=LinearInterpolate(A=Multiply(A=@n5.A) {constA=1, constB=0.5}, Alpha=@n7.Param1) {constAlpha=0.04})
  n29 = Clamp(None=Multiply(A=Power(Base=@n28) {constExponent=30}) {constB=20})
  n30 = Power(Base=@n1.R, Exp=LinearInterpolate(Alpha=@n7.VTiling) {constA=15, constAlpha=0}) {constExponent=50}
  n31 = LinearInterpolate(Alpha=@n7.VTiling) {constAlpha=0, constB=5}
  n32 = Add(A=Multiply(A=Add(A=@n27, B=@n29)) {constB=0.125}, B=Clamp(None=Multiply(A=@n30, B=@n31))) {constB=0.6}
  ShadingPathSwitch(
   Default=LinearInterpolate(A=StaticSwitchParameter(True=@n26, False=@n25) {parameterName=UseParticleColor}, Alpha=Clamp(None=@n32)) {constB=0}
   Mobile=ParticleColor().RGB
  ) {desc=Migrated from FeatureLevelSwitch}
  ShadingPathSwitch(Default=DepthFade(), Mobile=Constant() {r=1}) {desc=Migrated from FeatureLevelSwitch}
 ]
}
